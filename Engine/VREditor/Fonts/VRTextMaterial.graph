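# Masked, unlit material that renders VR Editor text in the vertex color with a dark outline.
# The red channel of the Font parameter (VRText_RobotoLarge by default) drives both outline and glyph shape.
# A SphereMask on that value, tuned by outline position, thickness and hardness parameters, blends vertex color to a constant vector for the outline.
# A steep Lerp remaps the same value into a sharp edge, multiplied by vertex alpha.
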Material /Engine/VREditor/Fonts/VRTextMaterial
graph {
 properties {blendMode=BLEND_Masked, pixelDepthOffsetMode=PDOM_Legacy, shadingModel=MSM_Unlit, usedShadingModels=Unlit}
 nodes [
  n1 = VertexColor()
  n2 = FontSampleParameter() {font=&/VRText_RobotoLarge, parameterName=Font}
  LinearInterpolate(
   A=@n1
   B=Constant3Vector()
   Alpha=SphereMask(
    A=@n2.R
    B=ScalarParameter() {defaultValue=0.42, parameterName=OutlineMask}
    Radius=ScalarParameter() {defaultValue=0.092, parameterName=OutlineThickness}
    Hardness=ScalarParameter() {defaultValue=0.9, parameterName=OutlineHardness}
   )
  )
  Multiply(A=LinearInterpolate(Alpha=@n2.R) {constA=-90, constB=90}, B=@n1.A)
 ]
}
